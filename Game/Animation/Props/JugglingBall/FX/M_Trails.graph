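# M_Trails is a translucent Default Lit material for juggling ball trail effects, usable on Niagara ribbons, sprites, mesh particles and beam trails.
# Particle color RGB, scaled down, feeds the MF_FX_EmissiveAdjustments function for glow.
# Particle alpha is multiplied by a sine of the half-tiled V coordinate, giving a texture-free soft falloff that peaks mid-strip and fades to both edges.
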Material /Game/Animation/Props/JugglingBall/FX/M_Trails
graph {
 properties {
  bUsedWithBeamTrails=true
  bUsedWithNiagaraMeshParticles=true
  bUsedWithNiagaraRibbons=true
  bUsedWithNiagaraSprites=true
  blendMode=BLEND_Translucent
  pixelDepthOffsetMode=PDOM_Legacy
  usedShadingModels=Default Lit
 }
 nodes [
  n1 = ParticleColor()
  MaterialFunctionCall(Emissive=Multiply(A=@n1.RGB) {constB=0.1}) {materialFunction=&MF_FX_EmissiveAdjustments}
  n2 = TextureCoordinate() {uTiling=0.5, vTiling=0.5}
  Multiply(A=@n1.A, B=Sine(None=ComponentMask(None=@n2) {g=true}))
 ]
}
