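# Physics asset for the Tomato Burger head juggling prop skeletal mesh, providing simple collision shapes for its bones.
# Five ball bones each carry an identical sphere body, and the root bone carries a small rotated capsule.
# No constraints link the bodies, so the asset defines no joints or limits between the balls and the root.
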PhysicsAsset /Game/Animation/Props/JugglingBall/Meshes/SK_Juggle_Head_Tomato_Burger_Physics
bodies {
 ball_01 [{shapeType=Sphere, radius=9, shapeName=ball_01_sphere}]
 ball_02 [{shapeType=Sphere, radius=9, shapeName=ball_02_sphere}]
 ball_03 [{shapeType=Sphere, radius=9, shapeName=ball_03_sphere}]
 ball_04 [{shapeType=Sphere, radius=9, shapeName=ball_04_sphere}]
 ball_05 [{shapeType=Sphere, radius=9, shapeName=ball_05_sphere}]
 root [{shapeType=Capsule, length=5.8045, radius=4.9103, center=<-0.0376 0.042 -0.1963>, rotation=<rot -37.5019 -90 8.5688>}]
}
constraints []
